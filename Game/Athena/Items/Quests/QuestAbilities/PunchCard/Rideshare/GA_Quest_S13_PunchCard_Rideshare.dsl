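; Season 13 Rideshare punch card quest ability that credits progress for a kill event involving the driver of the player's vehicle.
; On activation it requires the player to be in a vehicle, then polls that vehicle's driver on a looping timer to cache it.
; It listens for kill events whose target is the cached driver with matching player tags, sending a Rideshare quest stat.
; The ability ends when the Passenger movement tag is removed.
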(event Ability|EventActivateAbilityFromEvent
  (params (EventData :type Gameplay%20Event%20Data%20Structure%20%28by%20ref%29))
  (|CallParentFunction EventData)
  (bind _vehicle_occupied (Utilities|IsinVehicle (Variables|GAVars|Data|GetOwningPlayerPawn))
    (:true
      (Variables|GAVars|Data|SetCurrentVehicle _vehicle_occupied)
      (Utilities|Time|SetTimerbyEvent (delegate Custom|GetDriver) 0.4 true)
      (Ability|Tasks|WaitGameplayTagRemove "(TagName=\"MovementMode.Passenger\")"
        (:then
          (bind (_asynctaskproxy _payload) (Ability|Tasks|WaitGameplayEvent "(TagName=\"Event.Damage.Killed\")")
            (:EventReceived
              (if (and (Utilities|Equal(Object) (.Target (Utilities|Struct|BreakGameplayEventData _payload)) (Variables|GAVars|Data|GetCurrentDriver)) (GameplayTags|TagContainer|CompareTagContainertoOtherTags (Variables|GAVars|Setup|GetT_Player) (.TargetTags (Utilities|Struct|BreakGameplayEventData _payload))))
                ;[1 1 1] Send quest progress
                (Quest|SendComplexCustomStatEvent (Variables|GAVars|Data|GetOwningPlayerQuestManager) 0 "(GameplayTags=((TagName=\"Athena.Quests.Rideshare\")))" "(GameplayTags=)" 1)))))
        (:Removed
          (Ability|EndAbility))))
    (:false
      (Ability|EndAbility))))

(event Custom|GetDriver
  (Variables|GAVars|Data|SetCurrentDriver (Vehicle|GetDriver (Variables|GAVars|Data|GetCurrentVehicle))))
